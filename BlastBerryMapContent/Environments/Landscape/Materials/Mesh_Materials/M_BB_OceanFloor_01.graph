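# Opaque surface material for the Blast Berry ocean floor, a simplified underwater variant of the shared sand look.
# MF_BB_Sand_01 supplies raw diffuse, darkened by two constant multipliers, plus raw roughness, offset by the SandRoughUnderwaterAdd parameter and saturated.
# Specular comes from the sand function's material attributes.
# Tangent-space normals are off, and the world-space vertex normal drives Normal, so no sand normal detail is applied.
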Material /BlastBerryMapContent/Environments/Landscape/Materials/Mesh_Materials/M_BB_OceanFloor_01
graph {
 properties {
  bAutomaticallySetUsageInEditor=false
  bTangentSpaceNormal=false
  bUseMaterialAttributes=true
  pixelDepthOffsetMode=PDOM_Legacy
  usedShadingModels=Default Lit
 }
 nodes [
  n1 = MaterialFunctionCall() {materialFunction=/BlastBerryMapContent/Environments/Landscape/Materials/MaterialFunctions/MF_BB_Sand_01}
  n2 = Constant() {r=0.6}
  n3 = ScalarParameter() {defaultValue=0.5, group=Sand, parameterName=SandRoughUnderwaterAdd}
  SetMaterialAttributes!(
   MaterialAttributes
   Base Color=Multiply(A=Multiply(A=@n1.RawDiffuse, B=@n2), B=Constant() {r=0.8})
   Specular=GetMaterialAttributes(None=@n1.Result) {attributeGetTypes [#1, #2]}.Specular
   Roughness=Saturate(None=Add(A=@n1.RawRough, B=@n3))
   Normal=VertexNormalWS()
  ) {attributeSetTypes [
    #1
    #2
    #3
    #4
   ]}
 ]
}
